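# Translucent UI material that gives Legendary POI map text an animated, swirling shimmer, white to blue by default.
# Two panning samples of a swirl flow texture are blended with Light, Medium and Dark color parameters through nested constant-alpha lerps, some overdriven past 1.
# Panned UI UVs, scaled by SegmentCount and slightly rotated, feed a ping-pong function to form moving diagonal bands.
# Those bands are color-dodged over the swirl for moving highlights.
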Material /GoldenPOIUI/Content/UI/LegendaryPOI/Materials/M_UI_LegendaryPOI_MapText
graph {
 properties {bAutomaticallySetUsageInEditor=false, blendMode=BLEND_Translucent, materialDomain=MD_UI, usedShadingModels=Default Lit}
 nodes [
  n1 = Panner(Coordinate=MaterialFunctionCall() {materialFunction=&GetUserInterfaceUV}.Normalized UV) {speedX=-0.2}
  n2 = ScalarParameter() {defaultValue=1, parameterName=SegmentCount}
  n3 = ComponentMask(None=MaterialFunctionCall(UVs=Multiply(A=@n1, B=@n2), "Rotation Angle (0-1)"=Constant() {r=0.05}) {materialFunction=&CustomRotator}.Rotated Values) {r=true}
  n4 = TextureCoordinate() {uTiling=-1, vTiling=-0.5}
  n5 = Constant() {r=0.1}
  n6 = TextureSample(UVs=Panner(Coordinate=TextureCoordinate() {uTiling=0.5}, Speed=Constant() {r=0.08}) {speedX=-0.06, speedY=0.025}) {texture=/Game/Global/Textures/Default/Flowmaps/T_SwirlFlow_02_D}
  n7 = LinearInterpolate(
   A=TextureSample(UVs=Panner(Coordinate=@n4, Speed=@n5) {speedX=-0.06, speedY=0.025}) {texture=&T_SwirlFlow_02_D}.RGB
   B=@n6.RGB
  ) {constAlpha=3}
  n8 = LinearInterpolate(
   A=VectorParameter() {parameterName=Light, defaultValue=<rgba 1 1 1 1>}.RGB
   B=VectorParameter() {parameterName=Medium, defaultValue=<rgba 0.3515 0.855 1 1>}.RGB
  )
  n9 = Multiply(A=@n6.RGB) {constB=3}
  n10 = LinearInterpolate(A=VectorParameter() {parameterName=Dark, defaultValue=<rgba 0.0497 0.4735 1 1>}.RGB, B=@n8) {constAlpha=0.45}
  MaterialFunctionCall!(
   Blend=Multiply(A=MaterialFunctionCall!(Value=@n3, Scale, Height) {materialFunction=/FortUILibrary/MaterialFunctions/Basics/MF_UI_PingPong}.Result) {constB=0.6}
   Base=LinearInterpolate(A=LinearInterpolate(A=@n7, B=@n8), B=LinearInterpolate(A=@n9, B=@n10) {constAlpha=1.4}) {constAlpha=0.6, constB=1.5}
  ) {materialFunction=/Engine/Functions/Engine_MaterialFunctions03/Blends/Blend_ColorDodge}
 ]
}
